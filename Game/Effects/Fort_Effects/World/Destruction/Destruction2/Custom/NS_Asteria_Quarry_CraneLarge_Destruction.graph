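# Niagara destruction effect for the large quarry crane, bursting wood-mesh debris from scaffolding boxes and a boom-angle-rotated crane arm box.
# Debris launches from point and linear hit impulses that vary with instant death, then falls, tumbles and shrinks.
# GPU emitters add collision; CPU mobile variants omit it, with fewer scaffolding pieces.
# When bIsCurie is set, debris material parameters change and GPU spark sprites spawn from GPU debris particles.
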NiagaraSystem /Game/Effects/Fort_Effects/World/Destruction/Destruction2/Custom/NS_Asteria_Quarry_CraneLarge_Destruction
dependencies {
 usedDataInterfaces [&NiagaraDataInterfaceArrayFloat3, &NiagaraDataInterfaceCurve, &NiagaraDataInterfaceParticleRead]
 usedDynamicInputs [&ClampFloat, &Divide_Float]
 usedModules [
  &SystemState
  &self:SystemSpawnScript.SystemScriptSource.SystemScriptGraph.NiagaraNodeAssignment_0.SetVariables_#1
 ]
}
emitters {
 CraneArm {
  simTarget=GPUComputeSim
  emitterUpdateScript [
   &EmitterState()
   &SpawnBurst_Instantaneous(Spawn Count=150, Use Spawn Probability=false, UseLoopCountLimit=false)
  ]
  particleSpawnScript [
   /SetVariables(Particles.MeshIndex=&self:DetermineMeshProbability(
     Medium Mesh:NiagaraFloat=0.96
     Small Mesh:NiagaraFloat=0.5
     Randomness Mode:ENiagaraRandomnessMode=::NewEnumerator0
     Random Seed:NiagaraInt32=0
     Fixed Random Seed:NiagaraBool=false
    )) {isSetParameters=true, module=SetVariables_#2}
   &InitializeParticle(
    Color Maximum=<rgba 0.5395 0.2747 0.159 1>
    Color Minimum=<rgba 0.117 0.0561 0.0409 1>
    Color Mode=::NewEnumerator2
    Lifetime Mode=::NewEnumerator1
    Mass Mode=::NewEnumerator3
    Mesh Scale=&RandomRangeVector(
     Minimum=&SelectVectorFromArray(
      Direct Array Index=$Particles.MeshIndex
      Vector Selection Array=DataInterface() {propertyValues {FloatData [<1 1 1>, <2 2 2>, <0.75 0.75 0.75>]}}
     )
     Maximum=&SelectVectorFromArray(
      Direct Array Index=$Particles.MeshIndex
      Vector Selection Array=DataInterface() {propertyValues {FloatData [<1.5 1.5 1.5>, <3 3 3>, <1 1 1>]}}
     )
    )
    Mesh Scale Mode=::NewEnumerator4
    UsePositionOffset=false
    Write Material Random=true
   )
   &ShapeLocation(
    Additional Quaternion Rotation=true
    Additional Yaw / Pitch / Roll=&MakeVector(X=0, Y=$User.CraneBoomAngle, Z=0)
    Box Midpoint=<0 0.5 0.5>
    Box Size=<700 150 20>
    MaskSpawnGroup=false
    Offset=<450 0 -100>
    Offset Mode=::NewEnumerator0
    Override Seed=false
    Rotation Angle Type:ENiagara_AngleInput=::NewEnumerator0
    Rotation Mode=::NewEnumerator3
    Shape Primitive=::NewEnumerator2
    Surface Only=false
    UseOwnerScale=true
   )
   &InitialMeshOrientation(Use Rotation Vector=false)
   &PointForce(
    Force Falloff Distance:NiagaraFloat=$User.Radius
    Force Origin=$User.HitLocation
    Force Strength=&MakeCustomFloatFromBool(
     Bool=$User.bIsInstantDeath
     True Float=&Multiply_Float(A=&RandomRangeFloat(Minimum=1000, Maximum=1500), B=$System.ImpulseScale)
     False Float=&Multiply_Float(A=&RandomRangeFloat(Minimum=1750, Maximum=2250), B=$System.ImpulseScale)
    )
    Module.Use Falloff=true
    Module.UseLinearForce=false
   )
   &LinearForce(
    Coordinate Space:ENiagaraCoordinateSpace=::World
    Force=&Multiply_VectorByFloat(
     Vector=&Multiply_Vector(A=$User.HitVector, B=<1 1 0>)
     Float=&MakeCustomFloatFromBool(
      Bool=$User.bIsInstantDeath
      True Float=&Multiply_Float(A=&RandomRangeFloat(Minimum=1250, Maximum=1750), B=$System.ImpulseScale)
      False Float=&Multiply_Float(A=&RandomRangeFloat(Minimum=1500, Maximum=2000), B=$System.ImpulseScale)
     )
    )
   )
   &ApplyInitialForces()
  ]
  particleUpdateScript [
   &ParticleState()
   &ScaleMeshSize(Scale Factor=&VectorFromFloat(Value=&FloatFromCurve(FloatCurve=DataInterface() {propertyValues {Curve {keys [
          <at 0 0 RCIM_Constant>
          <at 0.05 1 RCIM_Constant>
          {interpMode=RCIM_Cubic, leaveTangentWeight=0.3333, tangentMode=RCTM_User, time=0.75, value=1}
          {arriveTangentWeight=0.3333, interpMode=RCIM_Cubic, tangentMode=RCTM_User, time=1, value=0}
         ]}}})))
   &GravityForce(Gravity=&Multiply_VectorByFloat(
     Vector=<0 0 -980>
     Float=&FloatFromCurve(FloatCurve=DataInterface() {propertyValues {Curve {keys [<at 0 0 RCIM_Constant>,<at 0.1 1>]}}}, CurveIndex=$Particles.Age)
    ))
   &Drag(
    Drag=&self:Gravity_Drag_Compensation(Drag Falloff:NiagaraFloat=2, Gravity Direction:Vector3f=<0 0 -1>)
    Use Linear Drag=true
    Use Rotational Drag=false
   )
   &Collision(
    Age Colliding Particles=false
    Friction During a Bounce=0.2
    GPU Collision Type=::NewEnumerator2
    Mesh Dimensions:Vector3f=&self:Spawned_Mesh_Dimensions(Wood Mesh 0:Vector3f=<2 31 3>, Wood Mesh 1:Vector3f=<18 4 1>, Wood Mesh 2:Vector3f=<224 16 9>)
    Module.Control Rest State Via Surface Normal=false
    Module.EnableMaxCollisionCount=false
    Module.Randomize Collision Normal Vector=true
    Particle Radius Scale=0.1
    Radius Calculation Type=::NewEnumerator1
    Restitution=0.15
   )
   &UpdateMeshOrientation(
    Rotation Rate=&FloatFromCurve(
     FloatCurve=DataInterface() {propertyValues {Curve {keys [{arriveTangent=3.7367, interpMode=RCIM_Cubic, leaveTangent=3.7367, tangentMode=RCTM_User, time=0.05, value=0}, <at 1 1 RCIM_Cubic>]}}}
     Scale Curve=&RandomRangeFloat(Minimum=0.1, Maximum=2)
    )
    Rotation Vector=&MakeCustomVectorFromBool(Bool=$StackContext.HasCollided, True Vector=<0 0 0>, False Vector=<1 1 1>)
   )
   &SolveForcesAndVelocity(Clamp Velocity=false, Limit Acceleration=false)
   &DynamicMaterialParameters(
    Index 0 Param 1=&MakeCustomFloatFromBool(
     Bool=$User.bIsCurie
     True Float=&FloatFromCurve(FloatCurve=DataInterface() {propertyValues {Curve {keys [
          {arriveTangent=-9.3075, interpMode=RCIM_Cubic, leaveTangent=-9.3075, tangentMode=RCTM_User, time=0, value=2}
          {arriveTangent=-2.0574, interpMode=RCIM_Cubic, leaveTangent=-2.0574, tangentMode=RCTM_User, time=0.2, value=0.625}
          <at 0.75 0 RCIM_Cubic>
         ]}}})
     False Float=0
    )
    Index 0 Param 2=&MakeCustomFloatFromBool(Bool=$User.bIsCurie, True Float=1, False Float=0)
    Module.Param0WriteEnabled=true
    Module.Param1WriteEnabled=true
    Module.Param2WriteEnabled=false
    Module.Param3WriteEnabled=false
   )
  ]
  rendererClasses [&NiagaraMeshRendererProperties]
  renderers [{rendererClass=&NiagaraMeshRendererProperties, rendererIndex=0}]
 }
 CraneArm_Mobile {
  emitterUpdateScript [
   &EmitterState()
   &SpawnBurst_Instantaneous(Spawn Count=150, Use Spawn Probability=false, UseLoopCountLimit=false)
  ]
  particleSpawnScript [
   /SetVariables(Particles.MeshIndex=&self:DetermineMeshProbability(
     Medium Mesh:NiagaraFloat=0.96
     Small Mesh:NiagaraFloat=0.5
     Randomness Mode:ENiagaraRandomnessMode=::NewEnumerator0
     Random Seed:NiagaraInt32=0
     Fixed Random Seed:NiagaraBool=false
    )) {isSetParameters=true, module=SetVariables_#2}
   &InitializeParticle(
    Color Maximum=<rgba 0.5395 0.2747 0.159 1>
    Color Minimum=<rgba 0.117 0.0561 0.0409 1>
    Color Mode=::NewEnumerator2
    Lifetime Mode=::NewEnumerator1
    Mass Mode=::NewEnumerator3
    Mesh Scale=&RandomRangeVector(
     Minimum=&SelectVectorFromArray(
      Direct Array Index=$Particles.MeshIndex
      Vector Selection Array=DataInterface() {propertyValues {FloatData [<1 1 1>, <2 2 2>, <0.75 0.75 0.75>]}}
     )
     Maximum=&SelectVectorFromArray(
      Direct Array Index=$Particles.MeshIndex
      Vector Selection Array=DataInterface() {propertyValues {FloatData [<1.5 1.5 1.5>, <3 3 3>, <1 1 1>]}}
     )
    )
    Mesh Scale Mode=::NewEnumerator4
    UsePositionOffset=false
    Write Material Random=true
   )
   &ShapeLocation(
    Additional Quaternion Rotation=true
    Additional Yaw / Pitch / Roll=&MakeVector(X=0, Y=$User.CraneBoomAngle, Z=0)
    Box Midpoint=<0 0.5 0.5>
    Box Size=<700 150 20>
    MaskSpawnGroup=false
    Offset=<450 0 -100>
    Offset Mode=::NewEnumerator0
    Override Seed=false
    Rotation Angle Type:ENiagara_AngleInput=::NewEnumerator0
    Rotation Mode=::NewEnumerator3
    Shape Primitive=::NewEnumerator2
    Surface Only=false
    UseOwnerScale=true
   )
   &InitialMeshOrientation(Use Rotation Vector=false)
   &PointForce(
    Force Falloff Distance:NiagaraFloat=$User.Radius
    Force Origin=$User.HitLocation
    Force Strength=&MakeCustomFloatFromBool(
     Bool=$User.bIsInstantDeath
     True Float=&Multiply_Float(A=&RandomRangeFloat(Minimum=1000, Maximum=1500), B=$System.ImpulseScale)
     False Float=&Multiply_Float(A=&RandomRangeFloat(Minimum=1750, Maximum=2250), B=$System.ImpulseScale)
    )
    Module.Use Falloff=true
    Module.UseLinearForce=false
   )
   &LinearForce(
    Coordinate Space:ENiagaraCoordinateSpace=::World
    Force=&Multiply_VectorByFloat(
     Vector=&Multiply_Vector(A=$User.HitVector, B=<1 1 0>)
     Float=&MakeCustomFloatFromBool(
      Bool=$User.bIsInstantDeath
      True Float=&Multiply_Float(A=&RandomRangeFloat(Minimum=1250, Maximum=1750), B=$System.ImpulseScale)
      False Float=&Multiply_Float(A=&RandomRangeFloat(Minimum=1500, Maximum=2000), B=$System.ImpulseScale)
     )
    )
   )
   &ApplyInitialForces()
  ]
  particleUpdateScript [
   &ParticleState()
   &ScaleMeshSize(Scale Factor=&VectorFromFloat(Value=&FloatFromCurve(FloatCurve=DataInterface() {propertyValues {Curve {keys [
          <at 0 0 RCIM_Constant>
          <at 0.05 1 RCIM_Constant>
          {interpMode=RCIM_Cubic, leaveTangentWeight=0.3333, tangentMode=RCTM_User, time=0.75, value=1}
          {arriveTangentWeight=0.3333, interpMode=RCIM_Cubic, tangentMode=RCTM_User, time=1, value=0}
         ]}}})))
   &GravityForce(Gravity=&Multiply_VectorByFloat(
     Vector=<0 0 -980>
     Float=&FloatFromCurve(FloatCurve=DataInterface() {propertyValues {Curve {keys [<at 0 0 RCIM_Constant>,<at 0.1 1>]}}}, CurveIndex=$Particles.Age)
    ))
   &Drag(
    Drag=&self:Gravity_Drag_Compensation(Drag Falloff:NiagaraFloat=2, Gravity Direction:Vector3f=<0 0 -1>)
    Use Linear Drag=true
    Use Rotational Drag=false
   )
   &UpdateMeshOrientation(
    Rotation Rate=&FloatFromCurve(
     FloatCurve=DataInterface() {propertyValues {Curve {keys [{arriveTangent=3.7367, interpMode=RCIM_Cubic, leaveTangent=3.7367, tangentMode=RCTM_User, time=0.05, value=0}, <at 1 1 RCIM_Cubic>]}}}
     Scale Curve=&RandomRangeFloat(Minimum=0.1, Maximum=2)
    )
    Rotation Vector=<1 1 1>
   )
   &SolveForcesAndVelocity(Clamp Velocity=false, Limit Acceleration=false)
   &DynamicMaterialParameters(
    Index 0 Param 1=&MakeCustomFloatFromBool(
     Bool=$User.bIsCurie
     True Float=&FloatFromCurve(FloatCurve=DataInterface() {propertyValues {Curve {keys [
          {arriveTangent=-9.3075, interpMode=RCIM_Cubic, leaveTangent=-9.3075, tangentMode=RCTM_User, time=0, value=2}
          {arriveTangent=-2.0574, interpMode=RCIM_Cubic, leaveTangent=-2.0574, tangentMode=RCTM_User, time=0.2, value=0.625}
          <at 0.75 0 RCIM_Cubic>
         ]}}})
     False Float=0
    )
    Index 0 Param 2=&MakeCustomFloatFromBool(Bool=$User.bIsCurie, True Float=1, False Float=0)
    Module.Param0WriteEnabled=true
    Module.Param1WriteEnabled=true
    Module.Param2WriteEnabled=false
    Module.Param3WriteEnabled=false
   )
  ]
  rendererClasses [&NiagaraMeshRendererProperties]
  renderers [{rendererClass=&NiagaraMeshRendererProperties, rendererIndex=0}]
 }
 Curie_Sparks {
  simTarget=GPUComputeSim
  emitterUpdateScript [
   &EmitterState(Life Cycle Mode=::NewEnumerator1, Loop Behavior=::NewEnumerator1, Loop Duration=0.3, UseLoopDelay=false)
   &SpawnParticlesFromOtherEmitter(
    Attribute Reader=DataInterface() {propertyValues {EmitterBinding {bindingMode=Other, emitterName=Scaffolding}}}
    Spawn Rate=&MakeCustomFloatFromBool(Bool=$User.bIsCurie, True Float=10, False Float=0)
    Spawn Rate Per Particle Cap=10
    Use Maximum Age=false
    Use Minimum Age=false
   )
  ]
  particleSpawnScript [
   &InitializeParticle(
    Color=<rgba 30 13.125 3 1>
    Lifetime Max=1
    Lifetime Min=0.15
    Lifetime Mode=::NewEnumerator1
    Sprite Size Max=<3.5 2>
    Sprite Size Min=<2.5 1>
    Sprite Size Mode=::NewEnumerator2
    UsePositionOffset=false
    Write Material Random=true
   )
   &SampleParticlesFromOtherEmitter(Velocity Application:ENiagara_DirectReadApplicationMode=::NewEnumerator0, Velocity Sampling=::NewEnumerator0)
  ]
  particleUpdateScript [
   &ParticleState()
   &GravityForce()
   /Niagara/Modules/Update/Forces/CurlNoiseForce(
    Noise Frequency=300
    Noise Strength=&Multiply_Float(
     A=$Particles.MaterialRandom
     B=&FloatFromCurve(
      FloatCurve=DataInterface() {propertyValues {Curve {keys [
          <at 0 4147.9209 RCIM_Cubic>
          {arriveTangent=998.7806, interpMode=RCIM_Cubic, leaveTangent=998.7806, time=0.897, value=2308.9619}
          {arriveTangent=839.8494, interpMode=RCIM_Cubic, leaveTangent=839.8494, time=2.2666, value=6411.7881}
          {arriveTangent=818.7506, interpMode=RCIM_Cubic, leaveTangent=818.7506, time=3.6727, value=4640.1284}
          {arriveTangent=-197.3881, interpMode=RCIM_Cubic, leaveTangent=-197.3881, time=4.7165, value=8417.627}
          {arriveTangent=-812.5138, interpMode=RCIM_Cubic, leaveTangent=-812.5138, time=6.6551, value=4051.4272}
          {arriveTangent=379.0706, interpMode=RCIM_Cubic, leaveTangent=379.0706, time=9.3392, value=4661.6182}
          {arriveTangent=-88.1153, interpMode=RCIM_Cubic, leaveTangent=-88.1153, time=11.8318, value=6013.7651}
          <at 15.169 4147.9209 RCIM_Cubic>
         ]}}}
      CurveIndex=&Add_Float(A=$Engine.System.Age, B=$Particles.Age)
      Scale Curve=0.7
     )
    )
    Pan Noise=true
    Pan Noise Field=<0 0 1000>
   )
   &Collision(
    Age Colliding Particles=false
    Friction=0.35
    GPU Collision Type=::NewEnumerator2
    Module.Control Rest State Via Surface Normal=false
    Module.EnableMaxCollisionCount=false
    Module.Randomize Collision Normal Vector=false
    Restitution=0.4
   )
   &Drag(Drag=0.8, Use Linear Drag=true, Use Rotational Drag=false)
   &SolveForcesAndVelocity(Clamp Velocity=false, Limit Acceleration=false)
   &ScaleSpriteSizeBySpeed(Max Scale Factor=<1 10>, Velocity Threshold=1300)
   &ScaleColor(
    Scale Alpha=&FloatFromCurve(FloatCurve=DataInterface() {propertyValues {Curve {keys [<at 0 0>,<at 0.1997 0.9273>,<at 1.0011 0.002>]}}})
    ScaleA=true
    ScaleRGB=true
   )
  ]
  rendererClasses [&NiagaraSpriteRendererProperties]
  renderers [{rendererClass=&NiagaraSpriteRendererProperties}]
 }
 Curie_Sparks001 {
  simTarget=GPUComputeSim
  emitterUpdateScript [
   &EmitterState(Life Cycle Mode=::NewEnumerator1, Loop Behavior=::NewEnumerator1, Loop Duration=0.3, UseLoopDelay=false)
   &SpawnParticlesFromOtherEmitter(
    Attribute Reader=DataInterface() {propertyValues {EmitterBinding {bindingMode=Other, emitterName=CraneArm}}}
    Spawn Rate=&MakeCustomFloatFromBool(Bool=$User.bIsCurie, True Float=10, False Float=0)
    Spawn Rate Per Particle Cap=10
    Use Maximum Age=false
    Use Minimum Age=false
   )
  ]
  particleSpawnScript [
   &InitializeParticle(
    Color=<rgba 30 13.125 3 1>
    Lifetime Max=1
    Lifetime Min=0.15
    Lifetime Mode=::NewEnumerator1
    Sprite Size Max=<3.5 2>
    Sprite Size Min=<2.5 1>
    Sprite Size Mode=::NewEnumerator2
    UsePositionOffset=false
    Write Material Random=true
   )
   &SampleParticlesFromOtherEmitter(Velocity Application:ENiagara_DirectReadApplicationMode=::NewEnumerator0, Velocity Sampling=::NewEnumerator0)
  ]
  particleUpdateScript [
   &ParticleState()
   &GravityForce()
   &CurlNoiseForce(
    Noise Frequency=300
    Noise Strength=&Multiply_Float(
     A=$Particles.MaterialRandom
     B=&FloatFromCurve(
      FloatCurve=DataInterface() {propertyValues {Curve {keys [
          <at 0 4147.9209 RCIM_Cubic>
          {arriveTangent=998.7806, interpMode=RCIM_Cubic, leaveTangent=998.7806, time=0.897, value=2308.9619}
          {arriveTangent=839.8494, interpMode=RCIM_Cubic, leaveTangent=839.8494, time=2.2666, value=6411.7881}
          {arriveTangent=818.7506, interpMode=RCIM_Cubic, leaveTangent=818.7506, time=3.6727, value=4640.1284}
          {arriveTangent=-197.3881, interpMode=RCIM_Cubic, leaveTangent=-197.3881, time=4.7165, value=8417.627}
          {arriveTangent=-812.5138, interpMode=RCIM_Cubic, leaveTangent=-812.5138, time=6.6551, value=4051.4272}
          {arriveTangent=379.0706, interpMode=RCIM_Cubic, leaveTangent=379.0706, time=9.3392, value=4661.6182}
          {arriveTangent=-88.1153, interpMode=RCIM_Cubic, leaveTangent=-88.1153, time=11.8318, value=6013.7651}
          <at 15.169 4147.9209 RCIM_Cubic>
         ]}}}
      CurveIndex=&Add_Float(A=$Engine.System.Age, B=$Particles.Age)
      Scale Curve=0.7
     )
    )
    Pan Noise=true
    Pan Noise Field=<0 0 1000>
   )
   &Collision(
    Age Colliding Particles=false
    Friction=0.35
    GPU Collision Type=::NewEnumerator2
    Module.Control Rest State Via Surface Normal=false
    Module.EnableMaxCollisionCount=false
    Module.Randomize Collision Normal Vector=false
    Restitution=0.4
   )
   &Drag(Drag=0.8, Use Linear Drag=true, Use Rotational Drag=false)
   &SolveForcesAndVelocity(Clamp Velocity=false, Limit Acceleration=false)
   &ScaleSpriteSizeBySpeed(Max Scale Factor=<1 10>, Velocity Threshold=1300)
   &ScaleColor(
    Scale Alpha=&FloatFromCurve(FloatCurve=DataInterface() {propertyValues {Curve {keys [<at 0 0>,<at 0.1997 0.9273>,<at 1.0011 0.002>]}}})
    ScaleA=true
    ScaleRGB=true
   )
  ]
  rendererClasses [&NiagaraSpriteRendererProperties]
  renderers [{rendererClass=&NiagaraSpriteRendererProperties}]
 }
 Scaffolding {
  simTarget=GPUComputeSim
  emitterUpdateScript [
   &EmitterState()
   &SpawnBurst_Instantaneous(Spawn Count=150, Use Spawn Probability=false, UseLoopCountLimit=false)
   &SpawnBurst_Instantaneous(Spawn Count=150, Spawn Group=1, Use Spawn Probability=false, UseLoopCountLimit=false) {module=SpawnBurst_Instantaneous001}
   &SpawnBurst_Instantaneous(Spawn Count=150, Spawn Group=2, Use Spawn Probability=false, UseLoopCountLimit=false) {module=SpawnBurst_Instantaneous002}
  ]
  particleSpawnScript [
   /SetVariables(Particles.MeshIndex=&self:DetermineMeshProbability(
     Medium Mesh:NiagaraFloat=0.96
     Small Mesh:NiagaraFloat=0.5
     Randomness Mode:ENiagaraRandomnessMode=::NewEnumerator0
     Random Seed:NiagaraInt32=0
     Fixed Random Seed:NiagaraBool=false
    )) {isSetParameters=true, module=SetVariables_#2}
   &InitializeParticle(
    Color Maximum=<rgba 0.5395 0.2747 0.159 1>
    Color Minimum=<rgba 0.117 0.0561 0.0409 1>
    Color Mode=::NewEnumerator2
    Lifetime Mode=::NewEnumerator1
    Mass Mode=::NewEnumerator3
    Mesh Scale=&RandomRangeVector(
     Minimum=&SelectVectorFromArray(
      Direct Array Index=$Particles.MeshIndex
      Vector Selection Array=DataInterface() {propertyValues {FloatData [<1 1 1>, <2 2 2>, <0.75 0.75 0.75>]}}
     )
     Maximum=&SelectVectorFromArray(
      Direct Array Index=$Particles.MeshIndex
      Vector Selection Array=DataInterface() {propertyValues {FloatData [<1.5 1.5 1.5>, <3 3 3>, <1 1 1>]}}
     )
    )
    Mesh Scale Mode=::NewEnumerator4
    UsePositionOffset=false
    Write Material Random=true
   )
   &ShapeLocation(
    Box Midpoint=<0.5 0.5 1>
    Box Size=<500 150 250>
    MaskSpawnGroup=true
    Offset=<0 150 0>
    Offset Mode=::NewEnumerator0
    Override Seed=false
    Shape Primitive=::NewEnumerator2
    Spawn Group Mask=1
    Surface Only=false
    UseOwnerScale=true
   )
   &ShapeLocation(
    Box Midpoint=<0.5 0.5 1>
    Box Size=<500 125 150>
    MaskSpawnGroup=true
    Offset=<0 275 -75>
    Offset Mode=::NewEnumerator0
    Override Seed=false
    Shape Primitive=::NewEnumerator2
    Surface Only=false
    UseOwnerScale=true
   ) {module=ShapeLocation001}
   &ShapeLocation(
    Box Midpoint=<0.5 0.5 1>
    Box Size=<500 30 200>
    MaskSpawnGroup=true
    Offset=<0 -75 0>
    Offset Mode=::NewEnumerator0
    Override Seed=false
    Shape Primitive=::NewEnumerator2
    Spawn Group Mask=2
    Surface Only=false
    UseOwnerScale=true
   ) {module=ShapeLocation002}
   &InitialMeshOrientation(Use Rotation Vector=false)
   &PointForce(
    Force Falloff Distance:NiagaraFloat=$User.Radius
    Force Origin=$User.HitLocation
    Force Strength=&MakeCustomFloatFromBool(
     Bool=$User.bIsInstantDeath
     True Float=&Multiply_Float(A=&RandomRangeFloat(Minimum=1000, Maximum=1500), B=$System.ImpulseScale)
     False Float=&Multiply_Float(A=&RandomRangeFloat(Minimum=1750, Maximum=2250), B=$System.ImpulseScale)
    )
    Module.Use Falloff=true
    Module.UseLinearForce=false
   )
   &LinearForce(
    Coordinate Space:ENiagaraCoordinateSpace=::World
    Force=&Multiply_VectorByFloat(
     Vector=&Multiply_Vector(A=$User.HitVector, B=<1 1 0>)
     Float=&MakeCustomFloatFromBool(
      Bool=$User.bIsInstantDeath
      True Float=&Multiply_Float(A=&RandomRangeFloat(Minimum=1250, Maximum=1750), B=$System.ImpulseScale)
      False Float=&Multiply_Float(A=&RandomRangeFloat(Minimum=1500, Maximum=2000), B=$System.ImpulseScale)
     )
    )
   )
   &ApplyInitialForces()
  ]
  particleUpdateScript [
   &ParticleState()
   &ScaleMeshSize(Scale Factor=&VectorFromFloat(Value=&FloatFromCurve(FloatCurve=DataInterface() {propertyValues {Curve {keys [
          <at 0 0 RCIM_Constant>
          <at 0.05 1 RCIM_Constant>
          {interpMode=RCIM_Cubic, leaveTangentWeight=0.3333, tangentMode=RCTM_User, time=0.75, value=1}
          {arriveTangentWeight=0.3333, interpMode=RCIM_Cubic, tangentMode=RCTM_User, time=1, value=0}
         ]}}})))
   &GravityForce(Gravity=&Multiply_VectorByFloat(
     Vector=<0 0 -980>
     Float=&FloatFromCurve(FloatCurve=DataInterface() {propertyValues {Curve {keys [<at 0 0 RCIM_Constant>,<at 0.1 1>]}}}, CurveIndex=$Particles.Age)
    ))
   &Drag(
    Drag=&self:Gravity_Drag_Compensation(Drag Falloff:NiagaraFloat=2, Gravity Direction:Vector3f=<0 0 -1>)
    Use Linear Drag=true
    Use Rotational Drag=false
   )
   &Collision(
    Age Colliding Particles=false
    Friction During a Bounce=0.2
    GPU Collision Type=::NewEnumerator2
    Mesh Dimensions:Vector3f=&self:Spawned_Mesh_Dimensions(Wood Mesh 0:Vector3f=<2 31 3>, Wood Mesh 1:Vector3f=<18 4 1>, Wood Mesh 2:Vector3f=<224 16 9>)
    Module.Control Rest State Via Surface Normal=false
    Module.EnableMaxCollisionCount=false
    Module.Randomize Collision Normal Vector=true
    Particle Radius Scale=0.1
    Radius Calculation Type=::NewEnumerator1
    Restitution=0.15
   )
   &UpdateMeshOrientation(
    Rotation Rate=&FloatFromCurve(
     FloatCurve=DataInterface() {propertyValues {Curve {keys [{arriveTangent=3.7367, interpMode=RCIM_Cubic, leaveTangent=3.7367, tangentMode=RCTM_User, time=0.05, value=0}, <at 1 1 RCIM_Cubic>]}}}
     Scale Curve=&RandomRangeFloat(Minimum=0.1, Maximum=2)
    )
    Rotation Vector=&MakeCustomVectorFromBool(Bool=$StackContext.HasCollided, True Vector=<0 0 0>, False Vector=<1 1 1>)
   )
   &SolveForcesAndVelocity(Clamp Velocity=false, Limit Acceleration=false)
   &DynamicMaterialParameters(
    Index 0 Param 1=&MakeCustomFloatFromBool(
     Bool=$User.bIsCurie
     True Float=&FloatFromCurve(FloatCurve=DataInterface() {propertyValues {Curve {keys [
          {arriveTangent=-9.3075, interpMode=RCIM_Cubic, leaveTangent=-9.3075, tangentMode=RCTM_User, time=0, value=2}
          {arriveTangent=-2.0574, interpMode=RCIM_Cubic, leaveTangent=-2.0574, tangentMode=RCTM_User, time=0.2, value=0.625}
          <at 0.75 0 RCIM_Cubic>
         ]}}})
     False Float=0
    )
    Index 0 Param 2=&MakeCustomFloatFromBool(Bool=$User.bIsCurie, True Float=1, False Float=0)
    Module.Param0WriteEnabled=true
    Module.Param1WriteEnabled=true
    Module.Param2WriteEnabled=false
    Module.Param3WriteEnabled=false
   )
  ]
  rendererClasses [&NiagaraMeshRendererProperties]
  renderers [{rendererClass=&NiagaraMeshRendererProperties, rendererIndex=0}]
 }
 Scaffolding_Mobile {
  emitterUpdateScript [
   &EmitterState()
   &SpawnBurst_Instantaneous(Spawn Count=50, Use Spawn Probability=false, UseLoopCountLimit=false)
   &SpawnBurst_Instantaneous(Spawn Count=50, Spawn Group=1, Use Spawn Probability=false, UseLoopCountLimit=false) {module=SpawnBurst_Instantaneous001}
   &SpawnBurst_Instantaneous(Spawn Count=50, Spawn Group=2, Use Spawn Probability=false, UseLoopCountLimit=false) {module=SpawnBurst_Instantaneous002}
  ]
  particleSpawnScript [
   /SetVariables(Particles.MeshIndex=&self:DetermineMeshProbability(
     Medium Mesh:NiagaraFloat=0.96
     Small Mesh:NiagaraFloat=0.5
     Randomness Mode:ENiagaraRandomnessMode=::NewEnumerator0
     Random Seed:NiagaraInt32=0
     Fixed Random Seed:NiagaraBool=false
    )) {isSetParameters=true, module=SetVariables_#2}
   &InitializeParticle(
    Color Maximum=<rgba 0.5395 0.2747 0.159 1>
    Color Minimum=<rgba 0.117 0.0561 0.0409 1>
    Color Mode=::NewEnumerator2
    Lifetime Mode=::NewEnumerator1
    Mass Mode=::NewEnumerator3
    Mesh Scale=&RandomRangeVector(
     Minimum=&SelectVectorFromArray(
      Direct Array Index=$Particles.MeshIndex
      Vector Selection Array=DataInterface() {propertyValues {FloatData [<1 1 1>, <2 2 2>, <0.75 0.75 0.75>]}}
     )
     Maximum=&SelectVectorFromArray(
      Direct Array Index=$Particles.MeshIndex
      Vector Selection Array=DataInterface() {propertyValues {FloatData [<1.5 1.5 1.5>, <3 3 3>, <1 1 1>]}}
     )
    )
    Mesh Scale Mode=::NewEnumerator4
    UsePositionOffset=false
    Write Material Random=true
   )
   &ShapeLocation(
    Box Midpoint=<0.5 0.5 1>
    Box Size=<500 150 250>
    MaskSpawnGroup=true
    Offset=<0 150 0>
    Offset Mode=::NewEnumerator0
    Override Seed=false
    Shape Primitive=::NewEnumerator2
    Spawn Group Mask=1
    Surface Only=false
    UseOwnerScale=true
   )
   &ShapeLocation(
    Box Midpoint=<0.5 0.5 1>
    Box Size=<500 125 150>
    MaskSpawnGroup=true
    Offset=<0 275 -75>
    Offset Mode=::NewEnumerator0
    Override Seed=false
    Shape Primitive=::NewEnumerator2
    Surface Only=false
    UseOwnerScale=true
   ) {module=ShapeLocation001}
   &ShapeLocation(
    Box Midpoint=<0.5 0.5 1>
    Box Size=<500 30 200>
    MaskSpawnGroup=true
    Offset=<0 -75 0>
    Offset Mode=::NewEnumerator0
    Override Seed=false
    Shape Primitive=::NewEnumerator2
    Spawn Group Mask=2
    Surface Only=false
    UseOwnerScale=true
   ) {module=ShapeLocation002}
   &InitialMeshOrientation(Use Rotation Vector=false)
   &PointForce(
    Force Falloff Distance:NiagaraFloat=$User.Radius
    Force Origin=$User.HitLocation
    Force Strength=&MakeCustomFloatFromBool(
     Bool=$User.bIsInstantDeath
     True Float=&Multiply_Float(A=&RandomRangeFloat(Minimum=1000, Maximum=1500), B=$System.ImpulseScale)
     False Float=&Multiply_Float(A=&RandomRangeFloat(Minimum=1750, Maximum=2250), B=$System.ImpulseScale)
    )
    Module.Use Falloff=true
    Module.UseLinearForce=false
   )
   &LinearForce(
    Coordinate Space:ENiagaraCoordinateSpace=::World
    Force=&Multiply_VectorByFloat(
     Vector=&Multiply_Vector(A=$User.HitVector, B=<1 1 0>)
     Float=&MakeCustomFloatFromBool(
      Bool=$User.bIsInstantDeath
      True Float=&Multiply_Float(A=&RandomRangeFloat(Minimum=1250, Maximum=1750), B=$System.ImpulseScale)
      False Float=&Multiply_Float(A=&RandomRangeFloat(Minimum=1500, Maximum=2000), B=$System.ImpulseScale)
     )
    )
   )
   &ApplyInitialForces()
  ]
  particleUpdateScript [
   &ParticleState()
   &ScaleMeshSize(Scale Factor=&VectorFromFloat(Value=&FloatFromCurve(FloatCurve=DataInterface() {propertyValues {Curve {keys [
          <at 0 0 RCIM_Constant>
          <at 0.05 1 RCIM_Constant>
          {interpMode=RCIM_Cubic, leaveTangentWeight=0.3333, tangentMode=RCTM_User, time=0.75, value=1}
          {arriveTangentWeight=0.3333, interpMode=RCIM_Cubic, tangentMode=RCTM_User, time=1, value=0}
         ]}}})))
   &GravityForce(Gravity=&Multiply_VectorByFloat(
     Vector=<0 0 -980>
     Float=&FloatFromCurve(FloatCurve=DataInterface() {propertyValues {Curve {keys [<at 0 0 RCIM_Constant>,<at 0.1 1>]}}}, CurveIndex=$Particles.Age)
    ))
   &Drag(
    Drag=&self:Gravity_Drag_Compensation(Drag Falloff:NiagaraFloat=2, Gravity Direction:Vector3f=<0 0 -1>)
    Use Linear Drag=true
    Use Rotational Drag=false
   )
   &UpdateMeshOrientation(
    Rotation Rate=&FloatFromCurve(
     FloatCurve=DataInterface() {propertyValues {Curve {keys [{arriveTangent=3.7367, interpMode=RCIM_Cubic, leaveTangent=3.7367, tangentMode=RCTM_User, time=0.05, value=0}, <at 1 1 RCIM_Cubic>]}}}
     Scale Curve=&RandomRangeFloat(Minimum=0.1, Maximum=2)
    )
    Rotation Vector=<1 1 1>
   )
   &SolveForcesAndVelocity(Clamp Velocity=false, Limit Acceleration=false)
   &DynamicMaterialParameters(
    Index 0 Param 1=&MakeCustomFloatFromBool(
     Bool=$User.bIsCurie
     True Float=&FloatFromCurve(FloatCurve=DataInterface() {propertyValues {Curve {keys [
          {arriveTangent=-9.3075, interpMode=RCIM_Cubic, leaveTangent=-9.3075, tangentMode=RCTM_User, time=0, value=2}
          {arriveTangent=-2.0574, interpMode=RCIM_Cubic, leaveTangent=-2.0574, tangentMode=RCTM_User, time=0.2, value=0.625}
          <at 0.75 0 RCIM_Cubic>
         ]}}})
     False Float=0
    )
    Index 0 Param 2=&MakeCustomFloatFromBool(Bool=$User.bIsCurie, True Float=1, False Float=0)
    Module.Param0WriteEnabled=true
    Module.Param1WriteEnabled=true
    Module.Param2WriteEnabled=false
    Module.Param3WriteEnabled=false
   )
  ]
  rendererClasses [&NiagaraMeshRendererProperties]
  renderers [{rendererClass=&NiagaraMeshRendererProperties, rendererIndex=0}]
 }
}
summary {
 systemName=NS_Asteria_Quarry_CraneLarge_Destruction
 emitters [
  {emitterName=Scaffolding, simTarget=GPUComputeSim, rendererClasses [&NiagaraMeshRendererProperties]}
  {emitterName=Scaffolding_Mobile, rendererClasses [&NiagaraMeshRendererProperties]}
  {emitterName=Curie_Sparks, simTarget=GPUComputeSim, rendererClasses [&NiagaraSpriteRendererProperties]}
  {emitterName=CraneArm, simTarget=GPUComputeSim, rendererClasses [&NiagaraMeshRendererProperties]}
  {emitterName=Curie_Sparks001, simTarget=GPUComputeSim, rendererClasses [&NiagaraSpriteRendererProperties]}
  {emitterName=CraneArm_Mobile, rendererClasses [&NiagaraMeshRendererProperties]}
 ]
 userVariables [
  User.bIsCurie {type=NiagaraBool}
  User.bIsInstantDeath {type=NiagaraBool}
  User.CraneBoomAngle {type=NiagaraFloat, defaultValue=0, description=""}
  User.DamageDealt {type=NiagaraFloat, defaultValue=50, description=""}
  User.HitLocation {type=Vector3f, description="", defaultValue {struct=Vector3f, value=<0 0 0>}}
  User.HitVector {type=Vector3f, description="", defaultValue {struct=Vector3f, value=<0 0 0>}}
  User.Radius {type=NiagaraFloat, defaultValue=100, description=""}
  User.Resource {type=NiagaraInt32}
 ]
}
