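# Niagara system for large door destruction, bursting mesh debris and trailing metal sparks.
# Resource emitters fill a box from User bounds and orientation, then push chunks along the hit vector and radially from the hit location.
# Spark emitters read debris particles through attribute readers to spawn short-lived, speed-stretched sprites.
# GPU versions add door scaling and randomized collision; Mobile copies skip both and use stronger forces, while bIsCurie drives debris material parameters in both.
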NiagaraSystem /Game/Effects/Fort_Effects/World/Destruction/Destruction2/Custom/NS_LargeDoor_Destruction
dependencies {
 usedDataInterfaces [
  &NiagaraDataInterfaceArrayFloat
  &NiagaraDataInterfaceCurve
  &NiagaraDataInterfaceArrayFloat3
  &NiagaraDataInterfaceParticleRead
  &NiagaraDataInterfaceVectorCurve
 ]
 usedDynamicInputs [&SimulationPosition]
 usedModules [&SystemState]
}
emitters {
 Metal_Sparks {
  simTarget=GPUComputeSim
  emitterUpdateScript [
   &EmitterState()
   &SpawnParticlesFromOtherEmitter(
    Attribute Reader=DataInterface() {propertyValues {EmitterBinding {bindingMode=Other, emitterName=Resource}}}
    Spawn Rate=50
    Use Maximum Age=false
    Use Minimum Age=false
   )
  ]
  particleSpawnScript [
   &InitializeParticle(
    Color=<rgba 30 26.9412 16 1>
    Lifetime=&UniformRangedFloat(Minimum=0.1, Maximum=0.45)
    Mass Mode=::NewEnumerator1
    Sprite Size=&UniformRangedVector2D(Minimum=<1 1>, Maximum=<2 2>)
    Sprite Size Mode=::NewEnumerator4
    UsePositionOffset=false
    Write Material Random=true
   ) {module=InitializeParticle001}
   &ShapeLocation(MaskSpawnGroup=false, Override Seed=false, Sphere Radius=5, UseOwnerScale=true)
   &SampleParticlesFromOtherEmitter(Applied Velocity Scale=0.5, Velocity Sampling=::NewEnumerator0)
   &CameraOffset(Camera Offset Amount=25)
   &AddVelocity(
    Module.Use Falloff Distance=false
    Override Seed=false
    Velocity Mode=::NewEnumerator1
    Velocity Origin=$Particles.Position
    Velocity Speed=500
   )
  ]
  particleUpdateScript [
   &ParticleState()
   &AccelerationForce(Acceleration=<0 0 -980>)
   &Drag(Drag=&RandomRangeFloat(Minimum=1, Maximum=3), Use Linear Drag=true, Use Rotational Drag=false)
   &SolveForcesAndVelocity(Clamp Velocity=false, Limit Acceleration=false)
   &ScaleSpriteSizeBySpeed(Max Scale Factor=<1 10>)
   &ScaleSpriteSize(Scale Factor=&Vector2DFromFloat(Value=&FloatFromCurve(FloatCurve=DataInterface() {propertyValues {Curve {keys [{arriveTangent=-6.725, interpMode=RCIM_Cubic, leaveTangent=-6.725, tangentMode=RCTM_User, time=0.65, value=1}, <at 1 0 RCIM_Cubic>]}}})))
   &ScaleColor(
    Scale RGB=&VectorFromCurve(VectorCurve=DataInterface() {propertyValues {
       XCurve {keys [<at 0.5 12 RCIM_Cubic>,<at 1 1 RCIM_Cubic>]}
       YCurve {keys [<at 0.5 6 RCIM_Cubic>,<at 1 1 RCIM_Cubic>]}
       ZCurve {keys [<at 0.5 2 RCIM_Cubic>,<at 1 1 RCIM_Cubic>]}
      }})
    ScaleA=true
    ScaleRGB=true
   )
  ]
  rendererClasses [&NiagaraSpriteRendererProperties]
  renderers [{rendererClass=&NiagaraSpriteRendererProperties}]
 }
 Metal_Sparks_Mobile {
  emitterUpdateScript [
   &EmitterState()
   &SpawnParticlesFromOtherEmitter(
    Attribute Reader=DataInterface() {propertyValues {EmitterBinding {bindingMode=Other, emitterName=Resource_Mobile}}}
    Use Maximum Age=false
    Use Minimum Age=false
   )
  ]
  particleSpawnScript [
   &InitializeParticle(
    Color=<rgba 30 26.9412 16 1>
    Lifetime=&UniformRangedFloat(Minimum=0.1, Maximum=0.45)
    Mass Mode=::NewEnumerator1
    Sprite Size=&UniformRangedVector2D(Minimum=<1 1>, Maximum=<2 2>)
    Sprite Size Mode=::NewEnumerator4
    UsePositionOffset=false
    Write Material Random=true
   ) {module=InitializeParticle001}
   &ShapeLocation(MaskSpawnGroup=false, Override Seed=false, Sphere Radius=5, UseOwnerScale=true)
   &SampleParticlesFromOtherEmitter(Applied Velocity Scale=0.5, Velocity Sampling=::NewEnumerator0)
   &CameraOffset(Camera Offset Amount=25)
   &AddVelocity(
    Module.Use Falloff Distance=false
    Override Seed=false
    Velocity Mode=::NewEnumerator1
    Velocity Origin=$Particles.Position
    Velocity Speed=500
   )
  ]
  particleUpdateScript [
   &ParticleState()
   &AccelerationForce(Acceleration=<0 0 -980>)
   &Drag(Drag=&RandomRangeFloat(Minimum=1, Maximum=3), Use Linear Drag=true, Use Rotational Drag=false)
   &SolveForcesAndVelocity(Clamp Velocity=false, Limit Acceleration=false)
   &ScaleSpriteSizeBySpeed(Max Scale Factor=<1 10>)
   &ScaleSpriteSize(Scale Factor=&Vector2DFromFloat(Value=&FloatFromCurve(FloatCurve=DataInterface() {propertyValues {Curve {keys [{arriveTangent=-6.725, interpMode=RCIM_Cubic, leaveTangent=-6.725, tangentMode=RCTM_User, time=0.65, value=1}, <at 1 0 RCIM_Cubic>]}}})))
   &ScaleColor(
    Scale RGB=&VectorFromCurve(VectorCurve=DataInterface() {propertyValues {
       XCurve {keys [<at 0.5 12 RCIM_Cubic>,<at 1 1 RCIM_Cubic>]}
       YCurve {keys [<at 0.5 6 RCIM_Cubic>,<at 1 1 RCIM_Cubic>]}
       ZCurve {keys [<at 0.5 2 RCIM_Cubic>,<at 1 1 RCIM_Cubic>]}
      }})
    ScaleA=true
    ScaleRGB=true
   )
  ]
  rendererClasses [&NiagaraSpriteRendererProperties]
  renderers [{rendererClass=&NiagaraSpriteRendererProperties}]
 }
 Resource {
  simTarget=GPUComputeSim
  emitterUpdateScript [
   &EmitterState()
   &SpawnBurst_Instantaneous(
    Spawn Count=&RandomRangeInt(Minimum=60, Maximum=120)
    Use Spawn Probability=false
    UseLoopCountLimit=false
   )
  ]
  particleSpawnScript [
   &self:InitialResourceSetup(Module.Override Randomness=false, Orientation:Quat4f=$User.Orientation) {module=InitialResourceSetup}
   &InitializeParticle(
    Color Mode=::NewEnumerator0
    Lifetime Max=5
    Lifetime Min=5
    Lifetime Mode=::NewEnumerator1
    Mesh Scale Mode=::NewEnumerator1
    Mesh Uniform Scale Max:NiagaraFloat=&SelectFloatFromArray(Direct Array Index=$Particles.MeshIndex, Float Selection Array=DataInterface() {propertyValues {FloatData [1,0.6,1]}})
    Mesh Uniform Scale Min:NiagaraFloat=&SelectFloatFromArray(Direct Array Index=$Particles.MeshIndex, Float Selection Array=DataInterface() {propertyValues {FloatData [0.5,0.15,0.3]}})
    UsePositionOffset=false
    Write Material Random=true
   )
   &ShapeLocation(
    Additional Quaternion Rotation=false
    Box Midpoint=<0.5 0.5 0>
    Box Size=&Multiply_Vector(
     A=&Multiply_VectorByFloat(Vector=$User.BoundsExtents, Float=2)
     B=&Multiply_Vector(A=$User.DoorScaling, B=<1 1 0.5>)
    )
    MaskSpawnGroup=false
    Override Seed=false
    Rotation Mode=::NewEnumerator3
    Rotation Quaternion=$Particles.Orientation
    Shape Primitive=::NewEnumerator2
    Surface Only=false
    UseOwnerScale=true
   )
   &InitialMeshOrientation(Use Rotation Vector=false)
   &LinearForce(
    Coordinate Space:ENiagaraCoordinateSpace=::World
    Force=&Multiply_VectorByFloat(
     Vector=&Multiply_Vector(A=$User.HitVector, B=<1 1 0>)
     Float=&RandomRangeFloat(Minimum=750, Maximum=1000)
    )
   )
   &PointForce(
    Force Falloff Distance:NiagaraFloat=$User.Radius
    Force Origin:NiagaraPosition=$User.HitLocation
    Force Strength=&RandomRangeFloat(Minimum=875, Maximum=1125)
    Module.Use Falloff=true
    Module.UseLinearForce=false
   )
   &ApplyInitialForces()
  ]
  particleUpdateScript [
   &ParticleState()
   &GravityForce(Gravity=&Multiply_VectorByFloat(
     Vector=<0 0 -980>
     Float=&FloatFromCurve(FloatCurve=DataInterface() {propertyValues {
        bOptimizeLUT=false
        Curve {keys [<at 0.05 0 RCIM_Cubic>, {arriveTangent=45.9003, interpMode=RCIM_Cubic, leaveTangent=45.9003, tangentMode=RCTM_User, time=0.1, value=1}]}
       }})
    ))
   &Drag(
    Drag=&self:Gravity_Drag_Compensation(Drag Falloff:NiagaraFloat=2, Gravity Direction:Vector3f=<0 0 -1>)
    Use Linear Drag=true
    Use Rotational Drag=true
   )
   &Collision(
    Age Colliding Particles=false
    Collision Enabled=&RandomBool()
    Friction During a Bounce=0.2
    GPU Collision Type=::NewEnumerator2
    KillOnCollision=false
    Mesh Dimensions:Vector3f=&SelectVectorFromArray(
     Direct Array Index=$Particles.MeshIndex
     Vector Selection Array=DataInterface() {propertyValues {FloatData [<32 32 6>, <52 52 13>, <16 7 9>]}}
    )
    Module.Control Rest State Via Surface Normal=false
    Module.EnableMaxCollisionCount=false
    Module.Randomize Collision Normal Vector=true
    Particle Radius Scale=0.1
    Radius Calculation Type=::NewEnumerator1
    Restitution=0.15
   )
   &SolveForcesAndVelocity(Clamp Velocity=false, Limit Acceleration=false)
   &UpdateMeshOrientation(Rotation Vector=&MakeCustomVectorFromBool(
     Bool=$Particles.HasCollided
     True Vector=<0 0 0>
     False Vector=&RandomRangeVector(Minimum=<-1 -1 -1>, Maximum=<1 1 1>)
    ))
   &ScaleMeshSize(Scale Factor=&VectorFromFloat(Value=&FloatFromCurve(
      FloatCurve=DataInterface() {propertyValues {Curve {keys [
          {interpMode=RCIM_Cubic, leaveTangentWeight=0.3333, tangentMode=RCTM_User, time=0.75, value=1}
          {arriveTangentWeight=0.3333, interpMode=RCIM_Cubic, tangentMode=RCTM_User, time=1, value=0}
         ]}}}
      Scale Curve=0.5
     )))
   &DynamicMaterialParameters(
    Index 0 Param 1=&MakeCustomFloatFromBool(
     Bool=$User.bIsCurie
     True Float=&FloatFromCurve(FloatCurve=DataInterface() {propertyValues {Curve {keys [
          {arriveTangent=-9.3075, interpMode=RCIM_Cubic, leaveTangent=-9.3075, tangentMode=RCTM_User, time=0, value=2}
          {arriveTangent=-2.0574, interpMode=RCIM_Cubic, leaveTangent=-2.0574, tangentMode=RCTM_User, time=0.2, value=0.625}
          <at 0.75 0 RCIM_Cubic>
         ]}}})
     False Float=0
    )
    Index 0 Param 2=&MakeCustomFloatFromBool(Bool=$User.bIsCurie, True Float=1, False Float=0)
    Index 0 Param 4=&RandomRangeFloat(Minimum=0, Maximum=1)
    Module.Index0Float0Write=true
    Module.Index0Float1Write=true
    Module.Index0Float2Write=true
    Module.Index0Float3Write=true
   )
  ]
  rendererClasses [&NiagaraMeshRendererProperties]
  renderers [{rendererClass=&NiagaraMeshRendererProperties, rendererIndex=0}]
 }
 Resource_Mobile {
  emitterUpdateScript [
   &EmitterState()
   &SpawnBurst_Instantaneous(
    Spawn Count=&RandomRangeInt(Minimum=60, Maximum=120)
    Use Spawn Probability=false
    UseLoopCountLimit=false
   )
  ]
  particleSpawnScript [
   &self:InitialResourceSetup(Module.Override Randomness=false, Orientation:Quat4f=$User.Orientation) {module=InitialResourceSetup}
   &InitializeParticle(
    Color Mode=::NewEnumerator0
    Lifetime Mode=::NewEnumerator1
    Mesh Scale Mode=::NewEnumerator1
    Mesh Uniform Scale Max:NiagaraFloat=&SelectFloatFromArray(Direct Array Index=$Particles.MeshIndex, Float Selection Array=DataInterface() {propertyValues {FloatData [1,0.6,1]}})
    Mesh Uniform Scale Min:NiagaraFloat=&SelectFloatFromArray(Direct Array Index=$Particles.MeshIndex, Float Selection Array=DataInterface() {propertyValues {FloatData [0.5,0.15,0.3]}})
    UsePositionOffset=false
    Write Material Random=true
   )
   &ShapeLocation(
    Additional Quaternion Rotation=false
    Box Size=&Multiply_VectorByFloat(Vector=$User.BoundsExtents, Float=2)
    MaskSpawnGroup=false
    Override Seed=false
    Rotation Mode=::NewEnumerator3
    Rotation Quaternion=$Particles.Orientation
    Shape Primitive=::NewEnumerator2
    Surface Only=false
    UseOwnerScale=true
   )
   &InitialMeshOrientation(Use Rotation Vector=false)
   &LinearForce(
    Coordinate Space:ENiagaraCoordinateSpace=::World
    Force=&Multiply_VectorByFloat(
     Vector=&Multiply_Vector(A=$User.HitVector, B=<1 1 0>)
     Float=&RandomRangeFloat(Minimum=1500, Maximum=2000)
    )
   )
   &PointForce(
    Force Falloff Distance:NiagaraFloat=$User.Radius
    Force Origin:NiagaraPosition=$User.HitLocation
    Force Strength=&RandomRangeFloat(Minimum=1750, Maximum=2250)
    Module.Use Falloff=true
    Module.UseLinearForce=false
   )
   &ApplyInitialForces()
  ]
  particleUpdateScript [
   &ParticleState()
   &GravityForce(Gravity=&Multiply_VectorByFloat(
     Vector=<0 0 -980>
     Float=&FloatFromCurve(FloatCurve=DataInterface() {propertyValues {
        bOptimizeLUT=false
        Curve {keys [<at 0.05 0 RCIM_Cubic>, {arriveTangent=45.9003, interpMode=RCIM_Cubic, leaveTangent=45.9003, tangentMode=RCTM_User, time=0.1, value=1}]}
       }})
    ))
   &Drag(
    Drag=&self:Gravity_Drag_Compensation(Drag Falloff:NiagaraFloat=2, Gravity Direction:Vector3f=<0 0 -1>)
    Use Linear Drag=true
    Use Rotational Drag=true
   )
   &SolveForcesAndVelocity(Clamp Velocity=false, Limit Acceleration=false)
   &UpdateMeshOrientation(Rotation Vector=&RandomRangeVector(Minimum=<-1 -1 -1>, Maximum=<1 1 1>))
   &ScaleMeshSize(Scale Factor=&VectorFromFloat(Value=&FloatFromCurve(FloatCurve=DataInterface() {propertyValues {Curve {keys [
          <at 0 0 RCIM_Constant>
          <at 0.05 1 RCIM_Constant>
          {interpMode=RCIM_Cubic, leaveTangentWeight=0.3333, tangentMode=RCTM_User, time=0.75, value=1}
          {arriveTangentWeight=0.3333, interpMode=RCIM_Cubic, tangentMode=RCTM_User, time=1, value=0}
         ]}}})))
   &DynamicMaterialParameters(
    Index 0 Param 1=&MakeCustomFloatFromBool(
     Bool=$User.bIsCurie
     True Float=&FloatFromCurve(FloatCurve=DataInterface() {propertyValues {Curve {keys [
          {arriveTangent=-9.3075, interpMode=RCIM_Cubic, leaveTangent=-9.3075, tangentMode=RCTM_User, time=0, value=2}
          {arriveTangent=-2.0574, interpMode=RCIM_Cubic, leaveTangent=-2.0574, tangentMode=RCTM_User, time=0.2, value=0.625}
          <at 0.75 0 RCIM_Cubic>
         ]}}})
     False Float=0
    )
    Index 0 Param 2=&MakeCustomFloatFromBool(Bool=$User.bIsCurie, True Float=1, False Float=0)
    Index 0 Param 4=&RandomRangeFloat(Minimum=0, Maximum=1)
    Module.Index0Float0Write=true
    Module.Index0Float1Write=true
    Module.Index0Float2Write=true
    Module.Index0Float3Write=true
   )
  ]
  rendererClasses [&NiagaraMeshRendererProperties]
  renderers [{rendererClass=&NiagaraMeshRendererProperties, rendererIndex=0}]
 }
}
summary {
 systemName=NS_LargeDoor_Destruction
 emitters [
  {emitterName=Resource, simTarget=GPUComputeSim, rendererClasses [&NiagaraMeshRendererProperties]}
  {emitterName=Metal_Sparks, simTarget=GPUComputeSim, rendererClasses [&NiagaraSpriteRendererProperties]}
  {emitterName=Resource_Mobile, rendererClasses [&NiagaraMeshRendererProperties]}
  {emitterName=Metal_Sparks_Mobile, rendererClasses [&NiagaraSpriteRendererProperties]}
 ]
 userVariables [
  User.bIsCurie {type=NiagaraBool}
  User.BoundsExtents {type=Vector3f, description="", defaultValue {struct=Vector3f, value=<0 0 0>}}
  User.DoorScaling {type=Vector3f, description="", defaultValue {struct=Vector3f, value=<1 1 1>}}
  User.HitLocation {type=NiagaraPosition, description="", defaultValue {}}
  User.HitVector {type=Vector3f, description="", defaultValue {struct=Vector3f, value=<1 0 0>}}
  User.Orientation {type=Quat4f, description="", defaultValue {struct=Quat4f, value=<0 0 0 1>}}
  User.Radius {type=NiagaraFloat, defaultValue=740, description=""}
 ]
}
